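# Physics asset for the SpeedDialBattle backpack mesh, giving collision bodies and joints to the pack and its heart and two letter bones.
# A box covers the pack, while capsules cover the dyn_bp_ bones, with two crossed capsules on letter1.
# Each of the three pieces constrains directly to the pack rather than forming a chain.
# All swing and twist axes are Locked, so the small stored angle limits allow no angular play.
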PhysicsAsset /BRCosmetics/Accessories/FORT_Backpacks/Backpack_M_MED_SpeedDialBattle/Meshes/M_MED_SpeedDialBattle_Pack_Physics
bodies {
 dyn_bp_heart [{
   shapeType=Capsule
   length=3.9713
   radius=3.1878
   shapeName=dyn_bp_heart_capsule
   center=<-0.858 -0.9418 0.3285>
   rotation=<rot 68.4634 172.4799 133.8835>
  }]
 dyn_bp_letter1 [
  {
   shapeType=Capsule
   length=28.8267
   radius=2.4967
   shapeName=dyn_bp_letter1_capsule
   center=<1.1663 -0.0114 0.0521>
   rotation=<rot 27.0038 101.1875 -60.3263>
  }
  {
   shapeType=Capsule
   length=17.924
   radius=2.5131
   shapeName=dyn_bp_letter1_capsule
   center=<-0.0099 -0.4357 0.5782>
   rotation=<rot 27.4301 100.2412 -118.3685>
  }
 ]
 dyn_bp_letter2 [{
   shapeType=Capsule
   length=18.1989
   radius=2.8441
   shapeName=dyn_bp_letter2_capsule
   center=<-0.2863 -0.195 0.7338>
   rotation=<rot -20.3525 90.6449 -86.0142>
  }]
 pack [{shapeType=Box, extentX=4.9236, extentY=25.6654, extentZ=28.4098, shapeName=pack_box, center=<0 4.5838 -6.0862>, rotation=<rot 0 -90 0>}]
}
constraints [
 {
  bone1Name=dyn_bp_heart
  bone2Name=pack
  swing1LimitDegrees=2.5
  swing1Motion=Locked
  swing2LimitDegrees=2.5
  swing2Motion=Locked
  twistLimitDegrees=2
  twistMotion=Locked
 }
 {
  bone1Name=dyn_bp_letter1
  bone2Name=pack
  swing1LimitDegrees=3
  swing1Motion=Locked
  swing2LimitDegrees=2
  swing2Motion=Locked
  twistLimitDegrees=2
  twistMotion=Locked
 }
 {
  bone1Name=dyn_bp_letter2
  bone2Name=pack
  swing1LimitDegrees=3
  swing1Motion=Locked
  swing2LimitDegrees=2
  swing2Motion=Locked
  twistLimitDegrees=2
  twistMotion=Locked
 }
]
