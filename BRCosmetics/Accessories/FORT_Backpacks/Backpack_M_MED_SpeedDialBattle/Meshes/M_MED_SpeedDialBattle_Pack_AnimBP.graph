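# Animation Blueprint for the SpeedDialBattle backpack cosmetic, adding secondary motion to the pack.
# A RigidBody simulation feeds additive bone offsets that shift the pack with spine rotation, sprinting and skydive or parachute states.
# The event graph also computes phase-offset sine waves that float two letter bones and a heart bone.
# Movement-based velocity, acceleration and gravity settings feed only an Anim Dynamics node left disconnected from the output pose.
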AnimBlueprint /BRCosmetics/Accessories/FORT_Backpacks/Backpack_M_MED_SpeedDialBattle/Meshes/M_MED_SpeedDialBattle_Pack_AnimBP
target_skeleton=&/M_MED_SpeedDialBattle_Pack_Skeleton
graphs [
 AnimGraph {nodes [
   Anim Dynamics - Bone: pack!(
    ComponentPose
    Alpha=1
    GravityOverride:Vector=Get Gravity_Override() -> (Gravity_Override:Vector)
    ComponentLinearAccScale:Vector=Get Linear_Acc_Scale() -> (Linear_Acc_Scale:Vector)
    ComponentLinearVelScale:Vector=Get Linear_Vel_Scale() -> (Linear_Vel_Scale:Vector)
    ExternalForce="0,0,0"
   ) -> (Pose) {
    bGravityOverrideInSimSpace=true
    bLinearSpring=true
    bUseGravityOverride=true
    linearSpringConstant=20
    simulationSpace=BoneRelative
    alphaBoolBlend {blendOption=Linear}
    alphaScaleBiasClamp {bInterpResult=true, interpSpeedDecreasing=500, interpSpeedIncreasing=500}
    boundBone {boneName=pack}
    componentAppliedLinearAccClamp=<50 60 130>
    gravityOverride {z=-980}
    physicsBodyDefinitions [{
      boundBone {boneName=pack}
      boxExtents=<20 8 20>
      constraintSetup {angularLimitsMax=<0 0 0>, angularLimitsMin=<0 0 0>, angularTarget=<0 0 0>}
      localJointOffset=<0 0 0>
     }]
    relativeSpaceBone {boneName=root}
   }
   n1 = RigidBody(ComponentPose=Local To Component(LocalPose=Input Pose())) {
    bOverrideWorldGravity=true
    overridePhysicsAsset=&/M_MED_SpeedDialBattle_Pack_Physics
    simulationSpace=BaseBoneSpace
    alphaBoolBlend {blendOption=Linear}
    baseBoneRef {boneName=root}
    componentAppliedLinearAccClamp=<3000 3000 3000>
    componentLinearAccScale=<0.1 0.1 0.1>
    componentLinearVelScale=<0.1 0.1 0.1>
    overrideWorldGravity {z=-200}
    simSpaceSettings {maxAngularAcceleration=50, maxAngularVelocity=5, maxLinearAcceleration=50, maxLinearVelocity=5, velocityScaleZ=0, worldAlpha=0.75}
   }
   n2 = "Transform (Modify) Bone - Bone: root"(ComponentPose=@n1, Translation="0,1.5,0", Rotation="0,0,0") {
    translationMode=BMM_Additive
    translationSpace=BCS_BoneSpace
    alphaBoolBlend {blendInTime=2, blendOption=Sinusoidal, blendOutTime=2}
    alphaScaleBiasClamp {bInterpResult=true}
    boneToModify {boneName=root}
    translation {y=1.5}
   }
   n3 = "Transform (Modify) Bone - Bone: pack"(ComponentPose=@n2, Translation="0,-2,5", Rotation="0,0,0") {
    translationMode=BMM_Additive
    translationSpace=BCS_BoneSpace
    alphaBoolBlend {blendInTime=2, blendOption=Sinusoidal, blendOutTime=2}
    alphaScaleBiasClamp {bInterpResult=true}
    boneToModify {boneName=pack}
    translation {y=-2, z=5}
   }
   n4 = "Transform (Modify) Bone - Bone: pack"(ComponentPose=@n3, Alpha=Get Backpack_Rotate_1() -> (Backpack_Rotate_1:double), Translation="-0.7,-0.5,0.7", Rotation="0,0,0") {
    translationMode=BMM_Additive
    translationSpace=BCS_BoneSpace
    alphaBoolBlend {blendInTime=2, blendOption=Sinusoidal, blendOutTime=2}
    alphaScaleBiasClamp {bInterpResult=true, interpSpeedDecreasing=15, interpSpeedIncreasing=15, scale=0.95}
    boneToModify {boneName=pack}
    translation=<-0.7 -0.5 0.7>
   }
   n5 = "Transform (Modify) Bone - Bone: pack"(ComponentPose=@n4, Alpha=Get Backpack_Rotate_2() -> (Backpack_Rotate_2:double), Translation="0.7,-0.5,0.7", Rotation="0,0,0") {
    translationMode=BMM_Additive
    translationSpace=BCS_BoneSpace
    alphaBoolBlend {blendInTime=2, blendOption=Sinusoidal, blendOutTime=2}
    alphaScaleBiasClamp {bInterpResult=true, interpSpeedDecreasing=15, interpSpeedIncreasing=15, scale=0.95}
    boneToModify {boneName=pack}
    translation=<0.7 -0.5 0.7>
   }
   n6 = "Transform (Modify) Bone - Bone: pack"!(ComponentPose=@n5, bAlphaBoolEnabled:bool=Get bIsSprinting(), Translation="0,-0.1,1", Rotation="0,0,0", Scale) -> (Pose) {
    alphaInputType=Bool
    translationMode=BMM_Additive
    translationSpace=BCS_BoneSpace
    alphaBoolBlend {blendInTime=0.6, blendOutTime=0.6}
    boneToModify {boneName=pack}
    translation {y=-0.1, z=1}
   }
   n7 = "Transform (Modify) Bone - Bone: pack"(ComponentPose=@n6, bAlphaBoolEnabled=Get Full_Skydive() -> (Full_Skydive:bool), Translation="0,-0.65,2", Rotation="0,0,0") {
    alphaInputType=Bool
    translationMode=BMM_Additive
    translationSpace=BCS_BoneSpace
    alphaBoolBlend {blendInTime=1, blendOption=Sinusoidal, blendOutTime=1}
    boneToModify {boneName=pack}
    translation {y=-0.65, z=2}
   }
   n8 = "Transform (Modify) Bone - Bone: dyn_bp_letter1"!(
    ComponentPose=@n7
    Alpha=1
    Translation_X:double=Get Wave_LetterX_1() -> (Wave_LetterX_1:double)
    Translation_Y:double=0
    Translation_Z:double=Get Wave_LetterX_2() -> (Wave_LetterX_2:double)
    Rotation="0,0,0"
    Scale="1,1,1"
   ) -> (Pose) {
    translationMode=BMM_Additive
    translationSpace=BCS_BoneSpace
    alphaBoolBlend {blendOption=Linear}
    boneToModify {boneName=dyn_bp_letter1}
    rotation {roll=25, yaw=15}
   }
   n9 = "Transform (Modify) Bone - Bone: dyn_bp_letter2"!(
    ComponentPose=@n8
    Alpha=1
    Translation_X:double=Get Wave_LetterO_1() -> (Wave_LetterO_1:double)
    Translation_Y:double=0
    Translation_Z:double=Get Wave_LetterO_2() -> (Wave_LetterO_2:double)
    Rotation="0,0,0"
    Scale="1,1,1"
   ) -> (Pose) {
    translationMode=BMM_Additive
    translationSpace=BCS_BoneSpace
    alphaBoolBlend {blendOption=Linear}
    boneToModify {boneName=dyn_bp_letter2}
    rotation {roll=25, yaw=15}
   }
   n10 = "Transform (Modify) Bone - Bone: dyn_bp_heart"!(
    ComponentPose=@n9
    Alpha=1
    Translation_X:double=Get Wave_Heart_1() -> (Wave_Heart_1:double)
    Translation_Y:double=0
    Translation_Z:double=Get Wave_Heart_2() -> (Wave_Heart_2:double)
    Rotation="0,0,0"
    Scale="1,1,1"
   ) -> (Pose) {
    translationMode=BMM_Additive
    translationSpace=BCS_BoneSpace
    alphaBoolBlend {blendOption=Linear}
    boneToModify {boneName=dyn_bp_heart}
    rotation {roll=25, yaw=15}
   }
   Output Pose(Result=Component To Local(ComponentPose=@n10)) {layerGroup=None}
  ]}
 EventGraph {nodes [
   Event BlueprintUpdateAnimation()
   > n1 = Sequence!(execute) -> (then_0, then_1, then_2:Exec)
   n2 = GetRealTimeSeconds()
   n3 = "Sin (Radians)"(A=float + float(A=float * float(A=@n2, B=1.5) {is=title K2Node_PromotableOperator}, B=0.2))
   > Set Wave_Heart_1(Wave_Heart_1:double=float + float(A=float * float(A=@n3, B=0.25), B=0.6)) -> (Output_Get:double)
   n4 = float + float(A=float * float(A=@n2, B=1.6), B=1.4)
   > Set Wave_Heart_2(Wave_Heart_2:double=float * float(A="Sin (Radians)"(A=@n4), B=0.25)) -> (Output_Get:double)
   n5 = float + float(A=float * float(A=@n2, B=1.5), B=2.4)
   > Set Wave_LetterX_1(Wave_LetterX_1:double=float * float(A="Sin (Radians)"(A=@n5), B=0.35)) -> (Output_Get:double)
   n6 = "Sin (Radians)"(A=float + float(A=float * float(A=@n2, B=1.6), B=3.6))
   > Set Wave_LetterX_2(Wave_LetterX_2:double=float + float(A=float * float(A=@n6, B=0.35), B=1)) -> (Output_Get:double)
   n7 = "Sin (Radians)"(A=float + float(A=float * float(A=@n2, B=1.5), B=4.8))
   > Set Wave_LetterO_1(Wave_LetterO_1:double=float + float(A=float * float(A=@n7, B=0.35), B=0.2)) -> (Output_Get:double)
   n8 = "Sin (Radians)"(A=float + float(A=float * float(A=@n2, B=1.6), B=6))
   > Set Wave_LetterO_2(Wave_LetterO_2:double=float + float(A=float * float(A=@n8, B=0.35), B=-1)) -> (Output_Get:double)
   > Set Full_Skydive(Full_Skydive:bool=OR Boolean!(A=Get bIsSkydiveFloating(), B=Get bIsSkydiveDiveMode(), C:bool=Get bIsParachuteOpened()) -> (ReturnValue)) -> (Output_Get:bool)
   n9 = Branch(execute=@n1.then_1, Condition=Get bIsCrouching())
   > Set Linear_Acc_Scale(Linear_Acc_Scale:Vector="0, 0, 0") -> (Output_Get:Vector)
   > n10 = Set Linear_Vel_Scale(Linear_Vel_Scale:Vector="10,11,13") -> (Output_Get:Vector)
   n11 = Get bIsCrouching()
   n12 = Branch(execute=[@n16.then, @n10.then], Condition=AND Boolean(A=Get bIsSprinting(), B=NOT Boolean(A=@n11)))
   > Set Linear_Acc_Scale(Linear_Acc_Scale="0.1,0.1,0.1")
   > n13 = Set Linear_Vel_Scale(Linear_Vel_Scale="9,11,13")
   n14 = Branch(execute=[@n13.then, @n15.then], Condition=OR Boolean!(A=Get bIsJumping(), B=Get bIsSkyDiving()) -> (ReturnValue))
   > Set Linear_Acc_Scale(Linear_Acc_Scale="0.1,0.1,0.1")
   > Set Linear_Vel_Scale(Linear_Vel_Scale="1.2,1.2,0.6")
   > Set Gravity_Override(Gravity_Override:Vector="0,-100,0") -> (Output_Get:Vector)
   Set Linear_Acc_Scale(execute=@n14.else, Linear_Acc_Scale="0.1,0.1,0.1")
   > Set Linear_Vel_Scale(Linear_Vel_Scale="0,0,1.3")
   > Set Gravity_Override(Gravity_Override="0, 0, 0")
   Set Linear_Acc_Scale(execute=@n12.else, Linear_Acc_Scale="0.1,0.1,0.1")
   > Set Linear_Vel_Scale(Linear_Vel_Scale="9,250,13")
   > n15 = Set Gravity_Override(Gravity_Override="0, 0, 0")
   Set Linear_Acc_Scale(execute=@n9.else, Linear_Acc_Scale="0.1,0.1,0.1")
   > Set Linear_Vel_Scale(Linear_Vel_Scale="0, 0, 0")
   > n16 = Set Gravity_Override(Gravity_Override="0, 0, 0")
   n17 = GetOwnerSkeletalMeshComponent()
   n18 = BreakRotator(InRot=BreakTransform(InTransform=GetDeltaTransformFromRefPose(self=@n17, BoneName=spine_05, BaseName=spine_02)).Rotation)
   Set Backpack_Rotate_1(execute=@n1.then_2, Backpack_Rotate_1:double=MapRangeClamped(Value=@n18, InRangeA=5, InRangeB=20)) -> (Output_Get:double)
   > Set Backpack_Rotate_2(Backpack_Rotate_2:double=MapRangeClamped(Value=@n18, InRangeA=-5, InRangeB=-20)) -> (Output_Get:double)
   TryGetPawnOwner()
  ]}
]
variables [
 Backpack_Rotate_1 {type=number}
 Backpack_Rotate_2 {type=number}
 Full_Skydive {type=boolean}
 Gravity_Override {type=Vector}
 Linear_Acc_Scale {type=Vector, default=<0.1 0.1 0.1>}
 Linear_Vel_Scale {type=Vector, default=<0 0 1.3>}
 Wave_Heart_1 {type=number, default=0.2081}
 Wave_Heart_2 {type=number, default=-0.2473}
 Wave_LetterO_1 {type=number, default=0.0073}
 Wave_LetterO_2 {type=number}
 Wave_LetterX_1 {type=number, default=0.3395}
 Wave_LetterX_2 {type=number, default=0.5955}
]
